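# Speech bubble widget that shows a player's latest chat message, hosted in an actor indicator.
# On construct it gives BubbleChatViewModel the player state from the parent FortActorIndicatorInterface widget's pawn, retrying on the view model's request if still unset.
# FillOutTextChatData hides the bubble when there are no messages, otherwise lists only the last one and tints the bubble for whispers.
# The Intro animation pops the bubble, scales its material tail and fades in text.
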WidgetBlueprint /Game/Athena/HUD/Team/BubbleChat/WBP_AthenaBubbleChat
animations [Intro {
  binding_count=3
  length_seconds=0.4569
  object=&self:Intro
  sequence {bindings [
    Overlay_BubbleContainer {
     id=#1
     possessed_class=Overlay
     tracks [
      {
       type=MovieSceneFloatTrack
       display_name=Render Opacity
       sections [{type=MovieSceneFloatSection, channels [None {type=MovieSceneScriptingFloatChannel, keys [[0.039,0],[3.0753,0],[6.9057,1]]}]}]
      }
      {
       type=MovieScene2DTransformTrack
       sections [{
         type=MovieScene2DTransformSection
         channels [
          Translation.X {type=MovieSceneScriptingFloatChannel}
          Translation.Y {type=MovieSceneScriptingFloatChannel, keys [[3.0753,15],[6.9057,0]]}
          Angle {type=MovieSceneScriptingFloatChannel}
          Scale.X {type=MovieSceneScriptingFloatChannel, keys [[3.0753,0.25],[5.4153,1.05],[6.3423,0.98],[6.9057,1]]}
          Scale.Y {type=MovieSceneScriptingFloatChannel, keys [[3.0753,0.25],[5.4153,1.05],[6.3423,0.98],[6.9057,1]]}
          Shear.X {type=MovieSceneScriptingFloatChannel}
          Shear.Y {type=MovieSceneScriptingFloatChannel}
         ]
        }]
      }
     ]
    }
    Default {
     id=#2
     possessed_class=Image
     tracks [{
       type=MovieSceneWidgetMaterialTrack
       sections [{type=MovieSceneParameterSection, channels [TailScale {type=MovieSceneScriptingFloatChannel, keys [[4.2473,0.5],[9.0997,1]]}]}]
      }]
    }
    ListView_Messages {
     id=#3
     possessed_class=CommonListView
     tracks [{
       type=MovieSceneFloatTrack
       display_name=Render Opacity
       sections [{type=MovieSceneFloatSection, channels [None {type=MovieSceneScriptingFloatChannel, keys [[5.2047,0],[9.177,1]]}]}]
      }]
    }
   ]}
 }]
graphs [
 EventGraph {nodes [
   n1 = OnBubbleChatVMTargetPlayerRequested() {is=title K2Node_CustomEvent}
   > SetTargetPlayer()
   Event Construct()
   > SetTargetPlayer()
   > n2 = Branch(Condition=HasTargetPlayerState!(self:Fort Bubble Chat View Model Object Reference=Get BubbleChatViewModel() -> (BubbleChatViewModel:Fort Bubble Chat View Model Object Reference)) -> (ReturnValue:bool))
   Bind Event to On Bubble Chat VMTarget Player Requested(execute=@n2.else, self:Fort Bubble Chat View Model Object Reference=Get BubbleChatViewModel(), Delegate=@n1) {is=title K2Node_AddDelegate}
   Event OnInitialized()
   > SetScrollbarVisibility!(
    execute
    self:List View Base Object Reference=Get ListView_Messages() -> (ListView_Messages:Common List View Object Reference)
    InVisibility:ESlateVisibility Enum=Collapsed
   ) -> (then)
  ]}
 FillOutTextChatData {nodes [
   FillOutTextChatData!() -> (then, Messages:Array of Fort Bubble Chat Message View Model Object References) {is=title K2Node_FunctionEntry}
   > ClearListItems(self=Get ListView_Messages())
   n1 = Is Not Empty!(TargetArray:Array of Fort Bubble Chat Message View Model Object References=Get Messages() -> (Messages:Array of Fort Bubble Chat Message View Model Object References)) -> (ReturnValue:bool)
   > SetVisibility(InVisibility=Select(Option 0=Collapsed, Option 1=SelfHitTestInvisible, Index=@n1))
   > PlayAnimation(InAnimation=Get Intro(), PlaybackSpeed=0.75)
   > Branch(Condition=@n1)
   n2 = Get Messages()
   n3 = "Get (a copy)"(
    Array:Array of Fort Bubble Chat Message View Model Object References=@n2
    Dimension 1=Last Index!(TargetArray:Array of Fort Bubble Chat Message View Model Object References=@n2) -> (ReturnValue=0)
   ) -> (Output:Fort Bubble Chat Message View Model Object Reference)
   > SetVisualStyle!(execute, self, IsWhisper:bool=Get IsWhisper!(self:Fort Bubble Chat Message View Model Object Reference=@n3) -> (IsWhisper:bool)) -> (then)
   > Set List Items!(
    execute
    self:List View Object Reference=Get ListView_Messages()
    InListItems:Array of Object References=Make Array("[0]":Object Reference=@n3) -> (Array:Array of Object References)
   ) -> (then) {is=title K2Node_CallFunction}
   > ScrollToBottom(self:List View Base Object Reference=Get ListView_Messages())
  ]}
 SetTargetPlayer {nodes [
   SetTargetPlayer() {is=title K2Node_FunctionEntry}
   > Is Valid(InputObject=Get BubbleChatViewModel())
   > n1 = Branch(Condition=HasTargetPlayerState(self=Get BubbleChatViewModel()))
   n2 = FindParentWidgetImplementingInterface!(
    execute=@n1.else
    StartingWidget:Widget Object Reference=Self-Reference()
    Interface:Interface Interface=/Script/FortniteGame.FortActorIndicatorInterface
   ) -> (then, ReturnValue:Widget Object Reference)
   > n3 = Cast To FortActorIndicatorInterface!(execute, Object=@n2) -> (then, CastFailed, AsFort Actor Indicator Interface:Fort Actor Indicator Interface Interface) {is=title K2Node_DynamicCast}
   > HandleIndicatorInterfaceFound!(
    execute
    self:Fort Bubble Chat View Model Object Reference=Get BubbleChatViewModel()
    InIndicatorInterface:Fort Actor Indicator Interface Interface=@n3
   ) -> (then)
   > n4 = GetIndicatedActor!(execute, self:Fort Actor Indicator Interface Interface=@n3) -> (then, ReturnValue:Actor Object Reference)
   > n5 = Cast To FortPlayerPawnAthena(Object=@n4)
   > n6 = Cast To FortPlayerStateAthena!(execute, Object=Get PlayerState!(self:Pawn Object Reference=@n5) -> (PlayerState:Player State Object Reference)) -> (then, CastFailed, AsFort Player State Athena:Fort Player State Athena Object Reference) {is=title K2Node_DynamicCast}
   > SetTargetPlayerState!(
    execute
    self:Fort Bubble Chat View Model Object Reference=Get BubbleChatViewModel()
    InTargetPlayerState:Fort Player State Athena Object Reference=@n6
   ) -> (then)
  ]}
 SetVisualStyle {nodes [
   n1 = SetVisualStyle!() -> (then, IsWhisper:bool=~) {is=title K2Node_FunctionEntry}
   > SetColorAndOpacity(
    self:Image Object Reference=Get Default() -> (Default:Image Object Reference)
    InColorAndOpacity:Linear Color Structure=Select(Option 0:Linear Color Structure=(R=0,G=0.0976,B=0.521,A=1), Option 1:Linear Color Structure=(R=0.1413,G=0.1912,B=0.8714,A=1), Index=@n1) -> (ReturnValue:Linear Color Structure)
   )
  ]}
]
variables []
widgets {
 info {inheritedWidgetCount=0, parentClass=&UserWidget, rootWidgetClass=&Overlay, widgetCount=4}
 widgets [
  Overlay_BubbleContainer {widgetClassPath=&Overlay, widget=Overlay_BubbleContainer}
  Default {
   widgetClassPath=&Image
   bIsVariable=true
   parent=Overlay_BubbleContainer
   widget=Default
   properties {
    visibility=SelfHitTestInvisible
    brush {
     bIsDynamicallyLoaded=false
     drawAs=Box
     imageType=FullColor
     mirroring=NoMirror
     resourceName=None
     resourceObject=/Game/Athena/HUD/Team/BubbleChat/Materials/MI_UI_Athena_SpeechBubble_OverheadDialogue_DefaultStyle
     tiling=NoTile
     imageSize=<94 94>
     margin=<ltrb 0.5 0.5 0.5 0.5>
     outlineSettings {
      bUseBrushTransparency=false
      roundingType=HalfHeightRadius
      width=0
      color {colorUseRule=UseColor_Specified, specifiedColor=<rgba 0 0 0 0>}
      cornerRadii=<0 0 0 0>
     }
     tintColor {colorUseRule=UseColor_Specified, specifiedColor=<rgba 1 1 1 1>}
     uVRegion {bIsValid=false, max=<0 0>, min=<0 0>}
    }
    colorAndOpacity=<rgba 0 0.0976 0.521 1>
    renderTransformPivot=<0.5 0.9>
   }
   slot {class=OverlaySlot, properties {horizontalAlignment=HAlign_Fill, verticalAlignment=VAlign_Fill, padding=<ltrb -23 -16 -23 -23>}}
  }
  CommonBorder_74 {
   widgetClassPath=&CommonBorder
   parent=Overlay_BubbleContainer
   widget=CommonBorder_74
   properties {padding=<ltrb 0 0 0 0>}
   slot {class=OverlaySlot, properties {padding=<ltrb 18 10 18 10>}}
  }
  ListView_Messages {
   widgetClassPath=&CommonListView
   bIsVariable=true
   parent=CommonBorder_74
   widget=ListView_Messages
   properties {
    allowOverscroll=false
    bAllowDragging=false
    bEnableRightClickScrolling=false
    bEnableScrollAnimation=false
    bEnableTouchScrolling=false
    bIsFocusable=false
    clipping=Inherit
    consumeMouseWheel=Never
    entryWidgetClass=&/WBP_AthenaBubbleChatEntry.WBP_AthenaBubbleChatEntry_C
    navigation=&self:WidgetTree.ListView_Messages.WidgetNavigation_0
    numDesignerPreviewEntries=1
    selectionMode=None
    visibility=SelfHitTestInvisible
    wheelScrollMultiplier=0
    scrollBarStyle {
     thickness=0
     draggedThumbImage {
      bIsDynamicallyLoaded=false
      drawAs=Box
      imageType=FullColor
      mirroring=NoMirror
      resourceName=../../../Engine/Content/Slate/Common/Scrollbar_Thumb.png
      resourceObject=null
      tiling=NoTile
      imageSize=<0 0>
      margin=<ltrb 0.25 0.25 0.25 0.25>
      outlineSettings {
       bUseBrushTransparency=false
       roundingType=HalfHeightRadius
       width=0
       color {colorUseRule=UseColor_Specified, specifiedColor=<rgba 0 0 0 0>}
       cornerRadii=<0 0 0 0>
      }
      tintColor {colorUseRule=UseColor_Specified, specifiedColor=<rgba 1 1 1 1>}
      uVRegion {bIsValid=false, max=<0 0>, min=<0 0>}
     }
     horizontalBackgroundImage {
      bIsDynamicallyLoaded=false
      drawAs=NoDrawType
      imageType=NoImage
      mirroring=NoMirror
      resourceName=None
      resourceObject=null
      tiling=NoTile
      imageSize=<0 0>
      margin=<ltrb 0 0 0 0>
      outlineSettings {
       bUseBrushTransparency=false
       roundingType=HalfHeightRadius
       width=0
       color {colorUseRule=UseColor_Specified, specifiedColor=<rgba 0 0 0 0>}
       cornerRadii=<0 0 0 0>
      }
      tintColor {colorUseRule=UseColor_Specified, specifiedColor=<rgba 1 1 1 1>}
      uVRegion {bIsValid=false, max=<0 0>, min=<0 0>}
     }
     horizontalBottomSlotImage {
      bIsDynamicallyLoaded=false
      drawAs=Image
      imageType=FullColor
      mirroring=NoMirror
      resourceName=../../../Engine/Content/Slate/Common/Scrollbar_Background_Horizontal.png
      resourceObject=null
      tiling=NoTile
      imageSize=<8 8>
      margin=<ltrb 0 0 0 0>
      outlineSettings {
       bUseBrushTransparency=false
       roundingType=HalfHeightRadius
       width=0
       color {colorUseRule=UseColor_Specified, specifiedColor=<rgba 0 0 0 0>}
       cornerRadii=<0 0 0 0>
      }
      tintColor {colorUseRule=UseColor_Specified, specifiedColor=<rgba 1 1 1 1>}
      uVRegion {bIsValid=false, max=<0 0>, min=<0 0>}
     }
     horizontalTopSlotImage {
      bIsDynamicallyLoaded=false
      drawAs=Image
      imageType=FullColor
      mirroring=NoMirror
      resourceName=../../../Engine/Content/Slate/Common/Scrollbar_Background_Horizontal.png
      resourceObject=null
      tiling=NoTile
      imageSize=<8 8>
      margin=<ltrb 0 0 0 0>
      outlineSettings {
       bUseBrushTransparency=false
       roundingType=HalfHeightRadius
       width=0
       color {colorUseRule=UseColor_Specified, specifiedColor=<rgba 0 0 0 0>}
       cornerRadii=<0 0 0 0>
      }
      tintColor {colorUseRule=UseColor_Specified, specifiedColor=<rgba 1 1 1 1>}
      uVRegion {bIsValid=false, max=<0 0>, min=<0 0>}
     }
     hoveredThumbImage {
      bIsDynamicallyLoaded=false
      drawAs=Box
      imageType=FullColor
      mirroring=NoMirror
      resourceName=../../../Engine/Content/Slate/Common/Scrollbar_Thumb.png
      resourceObject=null
      tiling=NoTile
      imageSize=<0 0>
      margin=<ltrb 0.25 0.25 0.25 0.25>
      outlineSettings {
       bUseBrushTransparency=false
       roundingType=HalfHeightRadius
       width=0
       color {colorUseRule=UseColor_Specified, specifiedColor=<rgba 0 0 0 0>}
       cornerRadii=<0 0 0 0>
      }
      tintColor {colorUseRule=UseColor_Specified, specifiedColor=<rgba 1 1 1 1>}
      uVRegion {bIsValid=false, max=<0 0>, min=<0 0>}
     }
     normalThumbImage {
      bIsDynamicallyLoaded=false
      drawAs=Box
      imageType=FullColor
      mirroring=NoMirror
      resourceName=../../../Engine/Content/Slate/Common/Scrollbar_Thumb.png
      resourceObject=null
      tiling=NoTile
      imageSize=<0 0>
      margin=<ltrb 0.25 0.25 0.25 0.25>
      outlineSettings {
       bUseBrushTransparency=false
       roundingType=HalfHeightRadius
       width=0
       color {colorUseRule=UseColor_Specified, specifiedColor=<rgba 0 0 0 0>}
       cornerRadii=<0 0 0 0>
      }
      tintColor {colorUseRule=UseColor_Specified, specifiedColor=<rgba 1 1 1 1>}
      uVRegion {bIsValid=false, max=<0 0>, min=<0 0>}
     }
     verticalBackgroundImage {
      bIsDynamicallyLoaded=false
      drawAs=NoDrawType
      imageType=NoImage
      mirroring=NoMirror
      resourceName=None
      resourceObject=null
      tiling=NoTile
      imageSize=<0 0>
      margin=<ltrb 0 0 0 0>
      outlineSettings {
       bUseBrushTransparency=false
       roundingType=HalfHeightRadius
       width=0
       color {colorUseRule=UseColor_Specified, specifiedColor=<rgba 0 0 0 0>}
       cornerRadii=<0 0 0 0>
      }
      tintColor {colorUseRule=UseColor_Specified, specifiedColor=<rgba 1 1 1 1>}
      uVRegion {bIsValid=false, max=<0 0>, min=<0 0>}
     }
     verticalBottomSlotImage {
      bIsDynamicallyLoaded=false
      drawAs=Image
      imageType=FullColor
      mirroring=NoMirror
      resourceName=../../../Engine/Content/Slate/Common/Scrollbar_Background_Vertical.png
      resourceObject=null
      tiling=NoTile
      imageSize=<8 8>
      margin=<ltrb 0 0 0 0>
      outlineSettings {
       bUseBrushTransparency=false
       roundingType=HalfHeightRadius
       width=0
       color {colorUseRule=UseColor_Specified, specifiedColor=<rgba 0 0 0 0>}
       cornerRadii=<0 0 0 0>
      }
      tintColor {colorUseRule=UseColor_Specified, specifiedColor=<rgba 1 1 1 1>}
      uVRegion {bIsValid=false, max=<0 0>, min=<0 0>}
     }
     verticalTopSlotImage {
      bIsDynamicallyLoaded=false
      drawAs=Image
      imageType=FullColor
      mirroring=NoMirror
      resourceName=../../../Engine/Content/Slate/Common/Scrollbar_Background_Vertical.png
      resourceObject=null
      tiling=NoTile
      imageSize=<8 8>
      margin=<ltrb 0 0 0 0>
      outlineSettings {
       bUseBrushTransparency=false
       roundingType=HalfHeightRadius
       width=0
       color {colorUseRule=UseColor_Specified, specifiedColor=<rgba 0 0 0 0>}
       cornerRadii=<0 0 0 0>
      }
      tintColor {colorUseRule=UseColor_Specified, specifiedColor=<rgba 1 1 1 1>}
      uVRegion {bIsValid=false, max=<0 0>, min=<0 0>}
     }
    }
   }
   slot {class=BorderSlot, properties {padding=<ltrb 0 0 0 0>}}
  }
 ]
}
